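; Camera modifier for the Katana dash charge that shifts and tightens the player's view while it is active.
; Overrides BlueprintModifyCamera to slide the view location along its right vector, nudge yaw slightly negative and narrow the FOV for a zoom-in.
; All three outputs lerp from the incoming view by the modifier's Alpha, so the effect follows the modifier's blend in and out.
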(fn BlueprintModifyCamera
  (params (DeltaTime :type Float%20%28single-precision%29) (ViewLocation :type Vector :default "0, 0, 0") (ViewRotation :type Rotator :default "0, 0, 0") (FOV :type Float%20%28single-precision%29))
  (bind (_viewrotation_roll _viewrotation_pitch _viewrotation_yaw) (Math|Rotator|BreakRotator (Variables|GetViewRotation_Roll)))
  (return (Math|Vector|Lerp(Vector) (Variables|GetViewLocation) (+ (Variables|GetViewLocation) (* (Math|Vector|GetRightVector (Variables|GetViewRotation)) 3e+01)) (Variables|CameraModifier|GetAlpha)) (Math|Rotator|Lerp(Rotator) (Variables|GetViewRotation) (Math|Rotator|MakeRotator _viewrotation_roll _viewrotation_pitch (+ _viewrotation_yaw -0.85)) (Variables|CameraModifier|GetAlpha)) (Math|Float|Lerp (Variables|GetFOV) (+ (Variables|GetFOV) -2e+01) (Variables|CameraModifier|GetAlpha))))
